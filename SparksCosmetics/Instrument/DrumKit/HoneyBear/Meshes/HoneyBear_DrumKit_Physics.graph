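# Physics asset for the HoneyBear drum kit cosmetic instrument, giving collision and joints to its drum base, two cymbals and two pairs of bee wings.
# One large capsule covers the drum base, with small spheres on each cymbal and offset spheres on the wings.
# Both cymbals join the base with free swing and twist.
# Bee A's wings hang from the left cymbal and bee B's from the right, each limited to a narrow single-axis swing.
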PhysicsAsset /SparksCosmetics/Instrument/DrumKit/HoneyBear/Meshes/HoneyBear_DrumKit_Physics
bodies {
 bone_L-cymbal [{shapeType=Sphere, radius=1, shapeName=bone_L-cymbal_sphere}]
 bone_R-cymbal [{shapeType=Sphere, radius=1, shapeName=bone_R-cymbal_sphere}]
 bone_drumbase [{
   shapeType=Capsule
   length=76
   radius=69.505
   shapeName=bone_drumbase_capsule
   center=<17.9103 -42.7611 36.6643>
   rotation=<rot 90 166.7373 166.7374>
  }]
 dyn_bee_a_wing_l [{shapeType=Sphere, radius=2, shapeName=dyn_bee_a_wing_l_sphere, center=<10 0 0>}]
 dyn_bee_a_wing_r [{shapeType=Sphere, radius=2, shapeName=dyn_bee_a_wing_r_sphere, center=<10 0 0>}]
 dyn_bee_b_wing_l [{shapeType=Sphere, radius=2, shapeName=dyn_bee_b_wing_l_sphere, center=<10 0 0>}]
 dyn_bee_b_wing_r [{shapeType=Sphere, radius=2, shapeName=dyn_bee_b_wing_r_sphere, center=<10 0 0>}]
}
constraints [
 {bone1Name=dyn_bee_b_wing_r, bone2Name=bone_R-cymbal, swing1LimitDegrees=15, swing2LimitDegrees=0, twistLimitDegrees=0}
 {bone1Name=dyn_bee_a_wing_l, bone2Name=bone_L-cymbal, swing1LimitDegrees=15, swing2LimitDegrees=0, twistLimitDegrees=0}
 {bone1Name=dyn_bee_a_wing_r, bone2Name=bone_L-cymbal, swing1LimitDegrees=15, swing2LimitDegrees=0, twistLimitDegrees=0}
 {bone1Name=dyn_bee_b_wing_l, bone2Name=bone_R-cymbal, swing1LimitDegrees=15, swing2LimitDegrees=0, twistLimitDegrees=0}
 {
  bone1Name=bone_L-cymbal
  bone2Name=bone_drumbase
  swing1LimitDegrees=45
  swing1Motion=Free
  swing2LimitDegrees=45
  swing2Motion=Free
  twistLimitDegrees=45
  twistMotion=Free
 }
 {
  bone1Name=bone_R-cymbal
  bone2Name=bone_drumbase
  swing1LimitDegrees=45
  swing1Motion=Free
  swing2LimitDegrees=45
  swing2Motion=Free
  twistLimitDegrees=45
  twistMotion=Free
 }
]
